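# Material function that builds a greyscale, flame-like silhouette from UVs, apparently for the RedSword plasma saw fire effect.
# The body is a SphereMask band centred at 0.5, cut at TipPositionOffset by a Ceil step and multiplied by MF_FX_TilingMethod with a per-channel clamp.
# The tip is a second SphereMask on UVs scaled around the tip point, giving an elliptical cap by default.
# A Max merges both, sharing Radius and Hardness inputs.
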MaterialFunction /Game/Effects/Fort_Effects/GameplayPulginsAssets/RedSword/Arrange/Mat/MF_PlasmaSaw_FireShape
graph {expressions {
  MaterialExpressionAdd_10 {
   class=MaterialExpressionAdd
   inputs [
    A {from=MaterialExpressionMultiply_19}
    B {from=MaterialExpressionConvert_7}
   ]
  }
  MaterialExpressionAdd_9 {
   class=MaterialExpressionAdd
   inputs [
    A {from=MaterialExpressionFunctionInput_1}
    B {from=MaterialExpressionFunctionInput_5}
   ]
   properties {constB=0}
  }
  MaterialExpressionCeil_3 {class=MaterialExpressionCeil, inputs [None {from=MaterialExpressionSubtract_2, output=""}], properties {}}
  MaterialExpressionConstant2Vector_1 {class=MaterialExpressionConstant2Vector, properties {g=1}}
  MaterialExpressionConstant_2 {class=MaterialExpressionConstant, properties {r=0.5}}
  MaterialExpressionConvert_6 {
   class=MaterialExpressionConvert
   inputs [None {from=MaterialExpressionAdd_9}]
   properties {
    nodeName=Break Float2
    convertInputs [{
      type=Vector2
      expressionInput {
       expression=&self:MaterialExpressionAdd_9
       inputName=None
       mask=0
       maskA=0
       maskB=0
       maskG=0
       maskR=0
       outputIndex=0
      }
     }]
    convertOutputs [{type=Scalar}, {type=Scalar}]
   }
  }
  MaterialExpressionConvert_7 {
   class=MaterialExpressionConvert
   inputs [
    None {from=MaterialExpressionFunctionInput_2}
    None {from=MaterialExpressionConstant_2}
   ]
   properties {
    nodeName=Make Vector2
    convertInputs [
     {
      type=Scalar
      expressionInput {
       expression=&self:MaterialExpressionFunctionInput_2
       inputName=None
       mask=0
       maskA=0
       maskB=0
       maskG=0
       maskR=0
       outputIndex=0
      }
     }
     {
      type=Scalar
      expressionInput {
       expression=&self:MaterialExpressionConstant_2
       inputName=None
       mask=0
       maskA=0
       maskB=0
       maskG=0
       maskR=0
       outputIndex=0
      }
     }
    ]
    convertOutputs [{type=Vector2}]
   }
  }
  MaterialExpressionFunctionInput_0 {
   class=MaterialExpressionFunctionInput
   inputs [Preview {from="", output=""}]
   properties {bUsePreviewValueAsDefault=true, inputName=ScaleTipUV, inputType=FunctionInput_Vector2, previewValue=<2 1 0 1>}
  }
  MaterialExpressionFunctionInput_1 {
   class=MaterialExpressionFunctionInput
   inputs [Preview {from=MaterialExpressionTextureCoordinate_1, output=""}]
   properties {bUsePreviewValueAsDefault=true, inputName=UV, inputType=FunctionInput_Vector2}
  }
  MaterialExpressionFunctionInput_2 {
   class=MaterialExpressionFunctionInput
   inputs [Preview {from="", output=""}]
   properties {bUsePreviewValueAsDefault=true, inputName=TipPositionOffset, inputType=FunctionInput_Scalar, previewValue=<0.15 0 0 1>}
  }
  MaterialExpressionFunctionInput_3 {
   class=MaterialExpressionFunctionInput
   inputs [Preview {from="", output=""}]
   properties {bUsePreviewValueAsDefault=true, inputName=Radius, inputType=FunctionInput_Scalar, previewValue=<0.2 0 0 1>}
  }
  MaterialExpressionFunctionInput_4 {
   class=MaterialExpressionFunctionInput
   inputs [Preview {from="", output=""}]
   properties {bUsePreviewValueAsDefault=true, inputName=Hardness, inputType=FunctionInput_Scalar, previewValue=<100 0 0 1>}
  }
  MaterialExpressionFunctionInput_5 {
   class=MaterialExpressionFunctionInput
   inputs [Preview {from="", output=""}]
   properties {bUsePreviewValueAsDefault=true, inputName=UV_Offset, inputType=FunctionInput_Vector2}
  }
  MaterialExpressionFunctionOutput_0 {class=MaterialExpressionFunctionOutput, inputs [None {from=MaterialExpressionMax_4, output=""}], properties {outputName=GreyScaleOut}}
  MaterialExpressionMaterialFunctionCall_11 {
   class=MaterialExpressionMaterialFunctionCall
   inputs [
    TexCoord {from=MaterialExpressionAdd_9}
    ClampChannel {from=MaterialExpressionConstant2Vector_1}
   ]
   properties {materialFunction=/Game/Effects/Fort_Effects/MaterialFunctions/MF_FX_TilingMethod}
  }
  MaterialExpressionMax_4 {
   class=MaterialExpressionMax
   inputs [
    A {from=MaterialExpressionSphereMask_7}
    B {from=MaterialExpressionMultiply_18}
   ]
  }
  MaterialExpressionMultiply_17 {
   class=MaterialExpressionMultiply
   inputs [
    A {from=MaterialExpressionMaterialFunctionCall_11, output=Result}
    B {from=MaterialExpressionSphereMask_6}
   ]
  }
  MaterialExpressionMultiply_18 {
   class=MaterialExpressionMultiply
   inputs [
    A {from=MaterialExpressionCeil_3}
    B {from=MaterialExpressionMultiply_17}
   ]
  }
  MaterialExpressionMultiply_19 {
   class=MaterialExpressionMultiply
   inputs [
    A {from=MaterialExpressionSubtract_3}
    B {from=MaterialExpressionFunctionInput_0}
   ]
  }
  MaterialExpressionReroute_2 {class=MaterialExpressionReroute, inputs [None {from=MaterialExpressionAdd_10}]}
  MaterialExpressionReroute_3 {class=MaterialExpressionReroute, inputs [None {from=MaterialExpressionConvert_7}]}
  MaterialExpressionSphereMask_6 {
   class=MaterialExpressionSphereMask
   inputs [
    A {from=MaterialExpressionConvert_6, output=""}
    B {from=MaterialExpressionConstant_2, output=""}
    Radius {from=MaterialExpressionFunctionInput_3, output=""}
    Hardness {from=MaterialExpressionFunctionInput_4, output=""}
   ]
   properties {}
  }
  MaterialExpressionSphereMask_7 {
   class=MaterialExpressionSphereMask
   inputs [
    A {from=MaterialExpressionReroute_2, output=""}
    B {from=MaterialExpressionReroute_3, output=""}
    Radius {from=MaterialExpressionFunctionInput_3, output=""}
    Hardness {from=MaterialExpressionFunctionInput_4, output=""}
   ]
   properties {}
  }
  MaterialExpressionSubtract_2 {
   class=MaterialExpressionSubtract
   inputs [
    A {from=MaterialExpressionConvert_6}
    B {from=MaterialExpressionFunctionInput_2}
   ]
  }
  MaterialExpressionSubtract_3 {
   class=MaterialExpressionSubtract
   inputs [
    A {from=MaterialExpressionAdd_9}
    B {from=MaterialExpressionConvert_7}
   ]
  }
  MaterialExpressionTextureCoordinate_1 {class=MaterialExpressionTextureCoordinate}
 }}
