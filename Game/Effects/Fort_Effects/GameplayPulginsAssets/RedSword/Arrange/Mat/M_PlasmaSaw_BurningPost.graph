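# Post-process material for the RedSword plasma saw burning effect, applying a directional screen blur within a world-space ellipsoid.
# A custom ray-ellipsoid intersection with Fresnel falloff masks the blur and, with radial noise, feeds the strength lerps.
# A custom weighted-tap loop blurs along a projected world direction, falling back to vertical when view-aligned, with temporal-dither jitter.
# A screen-aligned sphere mask around another world point restores the unblurred scene.
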Material /Game/Effects/Fort_Effects/GameplayPulginsAssets/RedSword/Arrange/Mat/M_PlasmaSaw_BurningPost
graph {
 properties {bAutomaticallySetUsageInEditor=false, materialDomain=MD_PostProcess, usedShadingModels=Default Lit}
 nodes [
  MaterialExpressionSceneTexture_3 = SceneTexture() {sceneTextureId=PPI_PostProcessInput0}
  n1 = ComponentMask(None=@MaterialExpressionSceneTexture_3.Color) {b=true, g=true, r=true}
  MaterialExpressionScreenPosition_0 = ScreenPosition()
  MaterialExpressionScalarParameter_4 = ScalarParameter() {defaultValue=4, parameterName=C}
  n2 = VectorParameter() {parameterName=BlurDirWorldSpace, defaultValue=<rgba 0 1 0 1>}
  n3 = Transform(None=@n2.RGB) {transformSourceType=TRANSFORMSOURCE_World, transformType=TRANSFORM_View}
  n4 = Abs(None=DotProduct(A=MaterialFunctionCall() {materialFunction=&CameraDirectionVector}.Result, B=@n2.RGB))
  n5 = ScalarParameter() {defaultValue=-0.7, group=BlurDir, parameterName=BlurDir_GrayAdjust_Add}
  n6 = Saturate(None=Multiply(A=Add(A=@n4, B=@n5), B=ScalarParameter() {defaultValue=3, group=BlurDir, parameterName=BlurDir_GrayAdjust_Mul}))
  n7 = ScalarParameter() {defaultValue=2, group=BlurDir, parameterName=BlurDir_GrayAdjust_Power}
  MaterialExpressionLinearInterpolate_1 = LinearInterpolate(A=ComponentMask(None=Normalize(VectorInput=@n3)) {g=true, r=true}, B=Constant2Vector() {g=1}, Alpha=Saturate(None=Power(Base=@n6, Exp=@n7)))
  n8 = VectorParameter() {parameterName=BlurStr, defaultValue=<rgba 0.005 0.01 0 1>}
  MaterialExpressionCameraPositionWS_0 = CameraPositionWS()
  MaterialExpressionMultiply_2 = Multiply(A=CameraVectorWS()) {constB=-1}
  MaterialExpressionVectorParameter_5 = VectorParameter() {parameterName=Center2, defaultValue=<rgba -2624 -3044.073 622 1>}
  MaterialExpressionVectorParameter_6 = VectorParameter() {parameterName=ER, defaultValue=<rgba 400 900 400 864406.25>}
  n9 = Custom!(
   rayOrigin=@MaterialExpressionCameraPositionWS_0
   rayDir=@MaterialExpressionMultiply_2
   EllipsoidCenter=@MaterialExpressionVectorParameter_5.RGB
   EllipsoidRadius=@MaterialExpressionVectorParameter_6.RGB
   MaxDist=@MaterialExpressionVectorParameter_6.RGB
  ) {
   code="float3 oc = rayOrigin - EllipsoidCenter;\r\n\r\nfloat3 invR2 = 1.0 / (EllipsoidRadius * EllipsoidRadius);\r\nfloat A = dot(rayDir * rayDir, invR2);\r\nfloat B = 2.0 * dot(oc * rayDir, invR2);\r\nfloat C = dot(oc * oc, invR2) - 1.0;\r\n\r\nfloat discriminant = B * B - 4.0 * A * C;\r\nif (discriminant < 0.0)\r\n{\r\n    return 0;\r\n}\r\n\r\nfloat sqrtD = sqrt(discriminant);\r\nfloat t0 = (-B - sqrtD) / (2.0 * A);\r\nfloat t1 = (-B + sqrtD) / (2.0 * A);\r\n\r\n\r\nfloat3 pCam = (rayOrigin - EllipsoidCenter) / EllipsoidRadius;\r\nfloat insideCheck = dot(pCam, pCam);   \r\nInside = (insideCheck < 1.0) ? 1.0 : 0.0;\r\n\r\n\r\nfloat t = -1.0;\r\nif (Inside < 0.5)\r\n{\r\n    \r\n    if (t0 > 0.0)      t = t0;\r\n    else if (t1 > 0.0) t = t1;\r\n}\r\nelse\r\n{\r\n    \r\n    if (t1 > 0.0)      t = t1;\r\n}\r\n\r\nif (t <= 0.0 || t > MaxDist)\r\n    return 0;\r\n\r\nDist = t;\r\nPos = rayOrigin + rayDir * t;\r\nreturn 1.0;"
   showCode=true
   additionalOutputs [{outputName=Dist, outputType=CMOT_Float1}, {outputName=Pos, outputType=CMOT_Float3}, {outputName=Inside, outputType=CMOT_Float1}]
   inputs [
    {
     inputName=rayOrigin
     input {expression=&self:MaterialExpressionCameraPositionWS_0}
    }
    {
     inputName=rayDir
     input {expression=&self:MaterialExpressionMultiply_2}
    }
    {
     inputName=EllipsoidCenter
     input {
      expression=&self:MaterialExpressionVectorParameter_5
      mask=1
      maskB=1
      maskG=1
      maskR=1
     }
    }
    {
     inputName=EllipsoidRadius
     input {
      expression=&self:MaterialExpressionVectorParameter_6
      mask=1
      maskB=1
      maskG=1
      maskR=1
     }
    }
    {
     inputName=MaxDist
     input {
      expression=&self:MaterialExpressionVectorParameter_6
      mask=1
      maskA=1
      outputIndex=4
     }
    }
   ]
  }
  n10 = Divide(
   A=Subtract(A=@n9.Pos, B=@MaterialExpressionVectorParameter_5.RGB)
   B=Multiply(A=@MaterialExpressionVectorParameter_6.RGB, B=@MaterialExpressionVectorParameter_6.RGB)
  )
  n11 = Add(
   A=DotProduct(A=Normalize(VectorInput=@n10), B=CameraVectorWS())
   B=ScalarParameter() {group=ESFresnel, parameterName=ESFresnel_GrayAdjust_Add}
  )
  n12 = ScalarParameter() {defaultValue=1, group=ESFresnel, parameterName=ESFresnel_GrayAdjust_Mul}
  n13 = Power(
   Base=Saturate(None=Multiply(A=@n11, B=@n12))
   Exp=ScalarParameter() {defaultValue=1, group=ESFresnel, parameterName=ESFresnel_GrayAdjust_Power}
  )
  n14 = LinearInterpolate(A=Multiply(A=@n9.return, B=Saturate(None=@n13)), Alpha=@n9.Inside)
  n15 = VectorParameter() {parameterName=SN, defaultValue=<rgba 0.15 3 0 1>}
  n16 = TextureObjectParameter() {
   group=VelocityTextureNoise
   parameterName=VelocityTextureNoise_Texture
   sortPriority=0
   texture=/Game/Effects/Fort_Effects/GameplayPulginsAssets/RedSword/Arrange/Tex/T_TempTexture_A
  }
  n17 = VectorParameter() {
   group=VelocityTextureNoise
   parameterName=VelocityTextureNoise_UVControl
   channelNames {a=VPan, b=UPan, g=VScale, r=UScale}
   defaultValue=<rgba 0.25 2 -3 0>
  }
  n18 = MaterialFunctionCall!(Center Point World Position=VectorParameter() {parameterName=Center, defaultValue=<rgba -2624 -4294.0728 622 1>}.RGB) {materialFunction=/Engine/Functions/Engine_MaterialFunctions02/UVs/WorldSpaceAlignedScreenCoordinates}
  n19 = MaterialFunctionCall(XY=MaterialFunctionCall(Vector or UVs=MaterialFunctionCall(XY=@n18.Offset and Scaled UVS) {materialFunction=&Swizzle}.YX) {materialFunction=&VectorToRadialValue}.Radial Coordinates) {materialFunction=&Swizzle}
  n20 = Multiply(
   A=Add(
    A=MaterialFunctionCall(Texture=@n16, UVControl_RGB=@n17.RGB, UVControl_A=@n17.RGB, "[Override]UVInput"=@n19.YX) {materialFunction=&FXCommon_TextureSample}.R
    B=ScalarParameter() {group=VelocityTextureNoise, parameterName=VelocityTextureNoise_GrayAdjust_Add}
   )
   B=ScalarParameter() {defaultValue=1, group=VelocityTextureNoise, parameterName=VelocityTextureNoise_GrayAdjust_Mul}
  )
  n21 = Saturate(None=Power(
    Base=Saturate(None=@n20)
    Exp=ScalarParameter() {defaultValue=1.1, group=VelocityTextureNoise, parameterName=VelocityTextureNoise_GrayAdjust_Power}
   ))
  MaterialExpressionMultiply_1 = Multiply(
   A=Multiply(A=LinearInterpolate(A=@n8.R, B=@n8.R, Alpha=@n14), B=LinearInterpolate(A=@n15.R, B=@n15.R, Alpha=@n21))
   B=ScalarParameter() {defaultValue=0.2, parameterName=Scale}
  )
  n22 = Custom!(None) {
   code=View.TemporalAAParams
   desc="\n*** Original source before expression upgrade ***\nView.TemporalAAParams\n*** Original source before expression upgrade ***\nFrame.TemporalAAParams"
   description=TemporalAA SampleID
   outputType=CMOT_Float2
  }
  MaterialExpressionAdd_2 = Add(
   A=MaterialFunctionCall!(TextureResolution=Constant2Vector() {g=1, r=1}) {materialFunction=/Engine/Functions/Engine_MaterialFunctions02/Texturing/ScreenAlignedPixelToPixelUVs}.Result
   B=ComponentMask(None=@n22) {r=true}
  )
  n23 = Custom!(p=@MaterialExpressionAdd_2) {
   code="((uint)(p.x) + 2 * (uint)(p.y)) % 5"
   desc="((uint)(p.x) + 2 * (uint)(p.y)) % 5"
   outputType=CMOT_Float1
   inputs [{
     inputName=p
     input {expression=&self:MaterialExpressionAdd_2}
    }]
  }
  MaterialExpressionDivide_1 = Divide(A=@n23) {constB=6}
  MaterialExpressionScalarParameter_5 = ScalarParameter() {defaultValue=1, parameterName=JitterStr}
  n24 = Custom!(
   Sampler=@MaterialExpressionSceneTexture_3.Color
   UV=@MaterialExpressionScreenPosition_0.ViewportUV
   Count=@MaterialExpressionScalarParameter_4
   BlurDir=@MaterialExpressionLinearInterpolate_1
   BlurStr=@MaterialExpressionMultiply_1
   Noise=@MaterialExpressionDivide_1
   JitterStr=@MaterialExpressionScalarParameter_5
  ) {
   code="\r\nfloat2 texelSize = View.ViewSizeAndInvSize.zw;\r\nfloat2 blurDirN = normalize(BlurDir);\r\nfloat2 stepUV   = blurDirN * BlurStr;\r\nfloat stepdist = BlurStr;\r\n\r\n\r\nint samples = (int)Count;\r\nsamples = clamp(samples, 1, 64);      \r\nint halfSamples = samples / 2;\r\n\r\nfloat4 accum = 0.0;\r\nfloat  weightSum = 0.0;\r\n[loop]\r\nfor (int i = -halfSamples; i <= halfSamples; i++)\r\n{\r\n    float t = (float)i;\r\n\r\n    float w = 1.0 - abs(t) / (halfSamples + 1e-4); \r\n    float jitter1 = (Noise-0.5)*stepdist*JitterStr;\r\n    float2 jitterOffset = float2(jitter1,jitter1) * blurDirN;\r\n    \r\n    float2 baseUV   = GetDefaultSceneTextureUV(Parameters, 14);\r\n    float2 uvOffset = baseUV + stepUV * t + jitterOffset;\r\n\r\n    accum += SceneTextureLookup(uvOffset, 14, false) * w;\r\n    weightSum += w;\r\n}\r\n\r\n\r\naccum /= max(weightSum, 1e-4);\r\n\r\nreturn accum;           "
   showCode=true
   inputs [
    {
     inputName=Sampler
     input {
      expression=&self:MaterialExpressionSceneTexture_3
      mask=1
      maskA=1
      maskB=1
      maskG=1
      maskR=1
     }
    }
    {
     inputName=UV
     input {expression=&self:MaterialExpressionScreenPosition_0}
    }
    {
     inputName=Count
     input {expression=&self:MaterialExpressionScalarParameter_4}
    }
    {
     inputName=BlurDir
     input {expression=&self:MaterialExpressionLinearInterpolate_1}
    }
    {
     inputName=BlurStr
     input {expression=&self:MaterialExpressionMultiply_1}
    }
    {
     inputName=Noise
     input {expression=&self:MaterialExpressionDivide_1}
    }
    {
     inputName=JitterStr
     input {expression=&self:MaterialExpressionScalarParameter_5}
    }
   ]
  }
  n25 = MaterialFunctionCall(Center Point World Position=VectorParameter() {parameterName=Center3, defaultValue=<rgba -2624 -3644.0728 622 1>}.RGB) {materialFunction=&WorldSpaceAlignedScreenCoordinates}
  n26 = Constant()
  n27 = ScalarParameter() {defaultValue=0.17, parameterName=Radius}
  n28 = ScalarParameter() {parameterName=Hardness}
  Reroute(None=LinearInterpolate(A=LinearInterpolate(A=@n1, B=@n24, Alpha=@n14), B=@n1, Alpha=SphereMask(A=@n25.Offset and Scaled UVS, B=@n26, Radius=@n27, Hardness=@n28)))
 ]
}
