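# Translucent, unlit, two-sided material for Niagara mesh particles that draws an energy spline strand, apparently for the RedSword plasma saw effect.
# World position offset combines SplineThicken, its width tapered by an edge mask, with panned soft noise displacement transformed from instance space and masked at tip and tail.
# A dynamic parameter offsets the noise UVs per particle.
# Emissive comes from an adjusted MajorColor, while opacity multiplies particle alpha by two UV edge fades.
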Material /Game/Effects/Fort_Effects/GameplayPulginsAssets/RedSword/Arrange/Mat/M_PlasmaSaw_EnergySpline
graph {
 properties {
  bAutomaticallySetUsageInEditor=false
  bUsedWithNiagaraMeshParticles=true
  blendMode=BLEND_Translucent
  previewMesh=/Engine/Transient.MaterialEditorMeshComponent_53
  shadingModel=MSM_Unlit
  twoSided=true
  usedShadingModels=Unlit
 }
 nodes [
  n1 = TextureObjectParameter() {group=WPONoise, parameterName=WPONoise_Texture, sortPriority=0, texture=/Game/Effects/Fort_Effects/Emotes/UltraLight/Texture/T_SoftNoise_BC7}
  n2 = DynamicParameter() {paramNames [TimeOffset,Param2,Param3,Param4]}
  n3 = VectorParameter() {group=WPONoise, parameterName=WPONoise_UVControl, channelNames {a=VPan, b=UPan, g=VScale, r=UScale}, defaultValue=<rgba 0.7 0.03 4 14.508>}
  n4 = AppendVector(
   A=Subtract(
    A=MaterialFunctionCall(Texture=@n1, AdditionalUVOffset=@n2.TimeOffset, UVControl_RGB=@n3.RGB, UVControl_A=@n3.RGB) {materialFunction=&FXCommon_TextureSample}.RG
    B=ScalarParameter() {defaultValue=0.5009, parameterName=OffsetV}
   ) {constB=0.5}
   B=Constant()
  )
  MaterialExpressionTextureCoordinate_0 = TextureCoordinate()
  n5 = Convert(None=@MaterialExpressionTextureCoordinate_0) {
   nodeName=Break Float2
   convertInputs [{
     type=Vector2
     expressionInput {
      expression=&self:MaterialExpressionTextureCoordinate_0
      inputName=None
      mask=0
      maskA=0
      maskB=0
      maskG=0
      maskR=0
      outputIndex=0
     }
    }]
   convertOutputs [{type=Scalar}, {type=Scalar}]
  }
  n6 = VectorParameter() {
   group=TipTailWPOMask
   parameterName=TipTailWPOMask_EdgeMask_InvertControl
   channelNames {a="", b=Power, g=Multiply, r=Offset}
   defaultValue=<rgba 0 1.5 1 1>
  }
  n7 = VectorParameter() {
   group=TipTailWPOMask
   parameterName=TipTailWPOMask_EdgeMask_Control
   channelNames {a="", b=Power, g=Multiply, r=Offset}
   defaultValue=<rgba 1 1 1 1>
  }
  n8 = MaterialFunctionCall!(UorV=@n5, Invert_EdgeControl=@n6.RGB, NoInvert_EdgeControl=@n7.RGB) {materialFunction=/Game/Effects/Niagara/MasterMaterials/Functions/Common/FXCommon_EdgeMask_Ex}
  n9 = ScalarParameter() {defaultValue=-28.076, parameterName=WPOStr}
  n10 = VectorParameter() {
   group=WdithTipTailFade
   parameterName=WdithTipTailFade_EdgeMask_InvertControl
   channelNames {a="", b=Power, g=Multiply, r=Offset}
   defaultValue=<rgba 0 3.3665 1 1>
  }
  n11 = VectorParameter() {
   group=WdithTipTailFade
   parameterName=WdithTipTailFade_EdgeMask_Control
   channelNames {a="", b=Power, g=Multiply, r=Offset}
   defaultValue=<rgba 0 1.6093 2 1>
  }
  n12 = Multiply(B=Multiply(
    A=MaterialFunctionCall(UorV=@n5, Invert_EdgeControl=@n10.RGB, NoInvert_EdgeControl=@n11.RGB) {materialFunction=&FXCommon_EdgeMask_Ex}.Mask
    B=ScalarParameter() {defaultValue=5.1142, parameterName=W}
   )) {constA=1}
  Add(
   A=Multiply(A=Transform(None=@n4) {transformSourceType=TRANSFORMSOURCE_Instance}, B=Multiply(A=@n8.Mask, B=@n9))
   B=MaterialFunctionCall(WidthBase=@n12, WidthTip=@n12, Expand U or V UV Channel=StaticBool()) {materialFunction=&SplineThicken}.WorldPositionOffset
  )
  MaterialFunctionCall(Emissive=Divide(A=VectorParameter() {parameterName=MajorColor, defaultValue=<rgba 0.4 1.4667 2 1>}.RGB)) {materialFunction=&MF_FX_EmissiveAdjustments}
  n13 = VectorParameter() {
   group=EdgeFadeU
   parameterName=EdgeFadeU_EdgeMask_InvertControl
   channelNames {a="", b=Power, g=Multiply, r=Offset}
   defaultValue=<rgba 1 40 0.1 1>
  }
  n14 = VectorParameter() {
   group=EdgeFadeU
   parameterName=EdgeFadeU_EdgeMask_Control
   channelNames {a="", b=Power, g=Multiply, r=Offset}
   defaultValue=<rgba -0.1536 2.1374 3 1>
  }
  n15 = VectorParameter() {group=EdgeF, parameterName=EdgeF_EdgeMask_Control, channelNames {a="", b=Power, g=Multiply, r=Offset}, defaultValue=<rgba 1 1 1 1>}
  n16 = MaterialFunctionCall!(UorV=@n5, EdgeControl=@n15.RGB) {materialFunction=/Game/Effects/Niagara/MasterMaterials/Functions/Common/FXCommon_EdgeMask_Simple}
  Multiply(
   A=ParticleColor().A
   B=Multiply(
    A=MaterialFunctionCall(UorV=@n5, Invert_EdgeControl=@n13.RGB, NoInvert_EdgeControl=@n14.RGB) {materialFunction=&FXCommon_EdgeMask_Ex}.Mask
    B=Multiply(A=@n16.Mask, B=@n15.A)
   )
  )
 ]
}
