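# M_PlasmaSaw_TestLight is a Default Lit test material in the RedSword effects folder, apparently for checking plasma saw lighting values.
# A constant of one passes through the MF_FX_EmissiveAdjustments function, and the result feeds a DebugScalarValues function as its number.
# No connection to a material output appears, so the graph looks like a leftover debug readout of the emissive adjustment.
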Material /Game/Effects/Fort_Effects/GameplayPulginsAssets/RedSword/Arrange/Mat/M_PlasmaSaw_TestLight
graph {
 properties {bAutomaticallySetUsageInEditor=false, usedShadingModels=Default Lit}
 nodes [MaterialFunctionCall(Number=MaterialFunctionCall(Emissive=Constant() {r=1}) {materialFunction=&MF_FX_EmissiveAdjustments}.Result) {materialFunction=&DebugScalarValues}]
}
