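# M_PlasmaSaw_Wind is an unlit, two-sided translucent Niagara material for the plasma saw's wind trail, tinted by particle color.
# The trail texture, distorted by panning noise, is multiplied by particle alpha, a hard dissolve, tip-tail edge masks, and squared or higher-power depth and near-camera fades.
# SplineThicken takes its width from a second edge mask.
# Dynamic parameters drive dissolve, noise strength, width and the time used for texture panning.
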Material /Game/Effects/Fort_Effects/GameplayPulginsAssets/RedSword/Arrange/Mat/M_PlasmaSaw_Wind
graph {
 properties {
  bAutomaticallySetUsageInEditor=false
  bUsedWithNiagaraMeshParticles=true
  bUsedWithNiagaraRibbons=true
  bUsedWithNiagaraSprites=true
  blendMode=BLEND_Translucent
  shadingModel=MSM_Unlit
  twoSided=true
  usedShadingModels=Unlit
 }
 nodes [
  n1 = ParticleColor()
  MaterialFunctionCall(Emissive=@n1.RGB) {materialFunction=&MF_FX_EmissiveAdjustments}
  n2 = DynamicParameter() {paramNames [Dissolve,Noise,Time,Width]}
  MaterialExpressionTextureCoordinate_0 = TextureCoordinate()
  n3 = Convert(None=@MaterialExpressionTextureCoordinate_0) {
   nodeName=Break Float2
   convertInputs [{
     type=Vector2
     expressionInput {
      expression=&self:MaterialExpressionTextureCoordinate_0
      inputName=None
      mask=0
      maskA=0
      maskB=0
      maskG=0
      maskR=0
      outputIndex=0
     }
    }]
   convertOutputs [{type=Scalar}, {type=Scalar}]
  }
  n4 = VectorParameter() {
   group=TipTailWidth
   parameterName=TipTailWidth_EdgeMask_InvertControl
   channelNames {a="", b=Power, g=Multiply, r=Offset}
   defaultValue=<rgba 0 15 1 1>
  }
  n5 = VectorParameter() {
   group=TipTailWidth
   parameterName=TipTailWidth_EdgeMask_Control
   channelNames {a="", b=Power, g=Multiply, r=Offset}
   defaultValue=<rgba 0 2 1 1>
  }
  n6 = Multiply(
   A=Multiply(A=@n2.Width, B=ScalarParameter() {defaultValue=80, parameterName=Width})
   B=MaterialFunctionCall!(UorV=@n3, Invert_EdgeControl=@n4.RGB, NoInvert_EdgeControl=@n5.RGB) {materialFunction=/Game/Effects/Niagara/MasterMaterials/Functions/Common/FXCommon_EdgeMask_Ex}.Mask
  )
  MaterialFunctionCall(WidthBase=@n6, WidthTip=@n6, Expand U or V UV Channel=StaticBool()) {materialFunction=&SplineThicken}
  n7 = MaterialFunctionCall(Distance=ScalarParameter() {defaultValue=150, parameterName=NearCameraFade}) {materialFunction=&Near_Camera_Fade}
  n8 = TextureObjectParameter() {
   group=Dissolve
   parameterName=Dissolve_Texture
   samplerType=SAMPLERTYPE_LinearColor
   sortPriority=0
   texture=/Engine/Functions/Engine_MaterialFunctions02/ExampleContent/Textures/OffsetNoiseDistanceFields_ForNormals
  }
  n9 = VectorParameter() {group=Dissolve, parameterName=Dissolve_UVControl, channelNames {a=VPan, b=UPan, g=VScale, r=UScale}, defaultValue=<rgba 0.7 0.2448 1 0>}
  n10 = MaterialFunctionCall(Texture=@n8, UVControl_RGB=@n9.RGB, UVControl_A=@n9.RGB, "[Override]TimeInput"=@n2.Time) {materialFunction=&FXCommon_TextureSample}
  n11 = Add(A=@n2.Dissolve, B=ScalarParameter() {defaultValue=-0.768, parameterName=DissolveOffset})
  n12 = Saturate(None=Multiply(A=Subtract(A=@n10.R, B=@n11), B=ScalarParameter() {defaultValue=100, parameterName=DissolveHardness}))
  n13 = TextureObjectParameter() {
   group=MainT
   parameterName=MainT_Texture
   sortPriority=0
   texture=/Game/Effects/Fort_Effects/GameplayPulginsAssets/RedSword/Arrange/Tex/T_PlasmaSaw_WindTrail_A
  }
  n14 = TextureObjectParameter() {
   group=MainNoise
   parameterName=MainNoise_Texture
   samplerType=SAMPLERTYPE_Normal
   sortPriority=0
   texture=/Game/Characters/Player/Female/Medium/Bodies/F_MED_Bistro_Astronaut/FX/T_GaussianNoise_Normal
  }
  n15 = VectorParameter() {group=MainNoise, parameterName=MainNoise_UVControl, channelNames {a=VPan, b=UPan, g=VScale, r=UScale}, defaultValue=<rgba 1 0.5 0.2 0>}
  n16 = VectorParameter() {
   group=MainNoise
   parameterName=MainNoise_NoiseControl
   sortPriority=60
   channelNames {a=VStrength, b=UStrength, g=VOffset, r=UOffset}
   defaultValue=<rgba 0 0 0.1 0.1>
  }
  n17 = Multiply(
   A=MaterialFunctionCall!(
    NoiseValue=MaterialFunctionCall(Texture=@n14, UVControl_RGB=@n15.RGB, UVControl_A=@n15.RGB, "[Override]TimeInput"=@n2.Time) {materialFunction=&FXCommon_TextureSample}.RG
    Control_RGB=@n16.RGB
    Control_A=@n16.RGB
   ) {materialFunction=/Game/Effects/Niagara/MasterMaterials/Functions/Common/FXCommon_NoiseControl}.NoiseValue
   B=@n2.Noise
  )
  n18 = VectorParameter() {group=MainT, parameterName=MainT_UVControl, channelNames {a=VPan, b=UPan, g=VScale, r=UScale}, defaultValue=<rgba 1 1 0.3 0>}
  n19 = Multiply(
   A=Multiply(
    A=@n1.A
    B=MaterialFunctionCall(Texture=@n13, AdditionalUVOffset=@n17, UVControl_RGB=@n18.RGB, UVControl_A=@n18.RGB, "[Override]TimeInput"=@n2.Time) {materialFunction=&FXCommon_TextureSample}.R
   )
   B=ScalarParameter() {defaultValue=1, parameterName=BO}
  )
  n20 = VectorParameter() {
   group=TipTailFade
   parameterName=TipTailFade_EdgeMask_InvertControl
   channelNames {a="", b=Power, g=Multiply, r=Offset}
   defaultValue=<rgba 1 1 1 1>
  }
  n21 = VectorParameter() {
   group=TipTailFade
   parameterName=TipTailFade_EdgeMask_Control
   channelNames {a="", b=Power, g=Multiply, r=Offset}
   defaultValue=<rgba 1 1 1 1>
  }
  n22 = Multiply(
   A=Multiply(A=@n7.Result, B=@n7.Result)
   B=Multiply(
    A=Multiply(A=@n12, B=@n19)
    B=MaterialFunctionCall(UorV=@n3, Invert_EdgeControl=@n20.RGB, NoInvert_EdgeControl=@n21.RGB) {materialFunction=&FXCommon_EdgeMask_Ex}.Mask
   )
  )
  n23 = DepthFade(FadeDistance=ScalarParameter() {defaultValue=50, parameterName=DF})
  n24 = Multiply(A=@n23, B=@n23)
  n25 = ScalarParameter() {defaultValue=150, parameterName=NearCameraFade}
  Multiply(
   A=Saturate(None=Multiply(A=@n22, B=@n24))
   B=Power(Base=MaterialFunctionCall(Distance=@n25) {materialFunction=&Near_Camera_Fade}.Result) {constExponent=5}
  )
 ]
}
